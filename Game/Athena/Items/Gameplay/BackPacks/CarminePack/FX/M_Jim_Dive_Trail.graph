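# Translucent unlit, two-sided beam-trail material, apparently for the Jim dive trail effect of the CarminePack back pack.
# Four differently panned shield-caustics samples on UV1 are blended and raised to a power as animated noise.
# Particle color lerps toward a hue-shifted copy through a UV mask, multiplied by the noise for emissive.
# Opacity combines particle alpha, noise and UV edge falloff, with depth and near-camera fades skipped on mobile.
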Material /Game/Athena/Items/Gameplay/BackPacks/CarminePack/FX/M_Jim_Dive_Trail
graph {
 properties {
  bUsedWithBeamTrails=true
  blendMode=BLEND_Translucent
  pixelDepthOffsetMode=PDOM_Legacy
  shadingModel=MSM_Unlit
  twoSided=true
  usedShadingModels=Unlit
 }
 nodes [
  n1 = ParticleColor()
  n2 = MaterialFunctionCall(Hue Shift Percentage=ScalarParameter() {defaultValue=0.5, parameterName="Hue Shift Percentage (S)"}, Texture=@n1.RGB) {materialFunction=&HueShift}
  n3 = ComponentMask(None=TextureCoordinate()) {g=true}
  n4 = Multiply(A=Power(Base=OneMinus(None=@n3)) {constExponent=1}, B=Power(Base=@n3) {constExponent=1})
  n5 = ComponentMask(None=TextureCoordinate()) {r=true}
  n6 = Clamp(None=Power(Base=OneMinus(None=@n5)))
  n7 = Clamp(None=Multiply(A=Clamp(None=@n4), B=@n6))
  n8 = TextureSample(UVs=Panner(Coordinate=TextureCoordinate() {coordinateIndex=1, vTiling=0.25}) {speedX=1}) {texture=&T_ShieldCaustics}
  n9 = TextureSample(UVs=Panner(Coordinate=TextureCoordinate() {coordinateIndex=1, uTiling=2, vTiling=0.5}) {speedX=0.6, speedY=-0.1}) {texture=&T_ShieldCaustics}
  n10 = TextureSample(UVs=Panner(Coordinate=TextureCoordinate() {coordinateIndex=1, uTiling=1.25, vTiling=0.5}) {speedX=1, speedY=0.15}) {texture=&T_ShieldCaustics}
  n11 = TextureSample(UVs=Panner(Coordinate=TextureCoordinate() {coordinateIndex=1, vTiling=0.75}) {speedX=0.7}) {texture=&T_ShieldCaustics}
  n12 = Power(
   Base=LinearInterpolate(A=LinearInterpolate(A=@n8.RGB, B=@n9.RGB), B=LinearInterpolate(A=@n10.RGB, B=@n11.RGB))
   Exp=ScalarParameter() {defaultValue=0.5, parameterName=Exp}
  ) {constExponent=1}
  MaterialFunctionCall(Emissive=Multiply(A=LinearInterpolate(A=@n1.RGB, B=@n2.Result, Alpha=@n7), B=@n12)) {materialFunction=&MF_FX_EmissiveAdjustments}
  n13 = Constant()
  n14 = Constant() {r=0.5}
  n15 = Multiply(A=ParticleColor().A, B=Add(A=LinearInterpolate(A=@n13, B=@n14, Alpha=@n6), B=@n12))
  n16 = DepthFade(Opacity=@n15, FadeDistance=ScalarParameter() {defaultValue=50, parameterName=FadeDistance})
  n17 = ComponentMask(None=TextureCoordinate()) {r=true}
  n18 = ComponentMask(None=TextureCoordinate()) {g=true}
  n19 = Multiply(A=Multiply(A=@n17, B=OneMinus(None=@n17)), B=Multiply(A=@n18, B=OneMinus(None=@n18)))
  n20 = ScalarParameter() {defaultValue=500, parameterName=NearFadeDistance}
  ShadingPathSwitch(
   Default=Multiply(
    A=Multiply(A=@n16, B=@n19)
    B=MaterialFunctionCall(Distance=@n20) {materialFunction=&Near_Camera_Fade}.Result
   )
   Mobile=Multiply(A=@n15, B=@n19)
  ) {desc=Migrated from FeatureLevelSwitch}
 ]
}
