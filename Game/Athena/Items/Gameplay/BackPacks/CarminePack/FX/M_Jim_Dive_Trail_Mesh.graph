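# Unlit translucent, two-sided mesh-particle material, apparently for the Jim dive trail in the CarminePack back bling FX.
# Layered panning chromatic inky-smoke samples form streaks, shaped and brightened along V by power-curve masks.
# Emissive combines particle color, time-of-day HDR multipliers and a Fresnel-indexed rainbow gradient, remapped through DeriveHDRfromLDR.
# Opacity combines streaks, V masks and particle alpha with inverse Fresnel edge fading, plus depth fade outside mobile.
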Material /Game/Athena/Items/Gameplay/BackPacks/CarminePack/FX/M_Jim_Dive_Trail_Mesh
graph {
 properties {
  bUsedWithMeshParticles=true
  bUsedWithNiagaraMeshParticles=true
  blendMode=BLEND_Translucent
  pixelDepthOffsetMode=PDOM_Legacy
  shadingModel=MSM_Unlit
  twoSided=true
  usedShadingModels=Unlit
 }
 nodes [
  n1 = TextureCoordinate() {uTiling=4}
  n2 = Multiply(A=Time(), B=ScalarParameter() {defaultValue=0.5, parameterName=PanSpeed})
  n3 = TextureCoordinate() {uTiling=6}
  n4 = LinearInterpolate(
   A=TextureSample(UVs=Panner(Coordinate=@n1, Time=@n2) {speedX=1.5, speedY=2.5}) {texture=&T_Inky_Smoke_Tile_Chromatic}.RGB
   B=TextureSample(UVs=Panner(Coordinate=@n3, Time=@n2) {speedX=-2, speedY=4}) {texture=&T_Inky_Smoke_Tile_Chromatic}.RGB
  )
  n5 = ComponentMask(None=TextureCoordinate()) {g=true}
  n6 = Power(
   Base=Multiply(A=@n4) {constB=1.5}
   Exp=LinearInterpolate(
    A=ScalarParameter() {parameterName=Streaks_LerpA}
    B=ScalarParameter() {defaultValue=0.4, parameterName=Streaks_LerpB}
    Alpha=OneMinus(None=@n5)
   ) {constA=0.6, constB=4}
  )
  n7 = ScalarParameter() {defaultValue=-1, parameterName=Desaturation}
  n8 = Desaturation(None=@n6, Fraction=@n7)
  n9 = Power(Base=ComponentMask(None=TextureCoordinate()) {g=true}, Exp=ScalarParameter() {defaultValue=1, parameterName=Alpha Contrast})
  n10 = LinearInterpolate(A=@n8, B=@n8, Alpha=Multiply(A=@n9, B=@n6))
  n11 = TextureCoordinate()
  n12 = Multiply(
   A=Power(Base=ComponentMask(None=@n11) {g=true}, Exp=ScalarParameter() {defaultValue=3, parameterName=Central Contast}) {constExponent=10}
   B=@n4
  )
  n13 = TextureCoordinate()
  n14 = TextureCoordinate()
  n15 = Add(
   A=TextureSample(UVs=Panner(Coordinate=@n13, Time=@n2) {speedX=-1, speedY=1.25}) {texture=&T_Inky_Smoke_Tile_Chromatic}.RGB
   B=TextureSample(UVs=Panner(Coordinate=@n14, Time=@n2) {speedX=1, speedY=2.5}) {texture=&T_Inky_Smoke_Tile_Chromatic}.RGB
  )
  n16 = Add(A=Add(A=@n12, B=Clamp(None=@n10) {maxDefault=5}), B=Multiply(A=@n10, B=Desaturation(None=@n15, Fraction=@n7)))
  n17 = Fresnel(ExponentIn=ScalarParameter() {defaultValue=0.05, parameterName=Fresnel Exp_1}) {baseReflectFraction=0.01}
  n18 = Multiply(
   A=TextureSample(UVs=@n17) {texture=&/T_Gradient_Rainbow_Jim}.RGB
   B=ScalarParameter() {defaultValue=0.25, parameterName=RainbowCololr}
  )
  n19 = ScalarParameter() {defaultValue=2, parameterName=HDRMult_3}
  n20 = ScalarParameter() {defaultValue=2, parameterName=HDRMult_2}
  n21 = ScalarParameter() {defaultValue=1, parameterName=HDRMult_1}
  n22 = ScalarParameter() {defaultValue=0.5, parameterName=HDRMult}
  n23 = Multiply(
   A=ParticleColor().RGB
   B=Multiply(
    A=@n16
    B=MaterialFunctionCall(Color_Morning=@n19, Color_Day=@n20, Color_Evening=@n21, Color_Night=@n22) {materialFunction=&Color_Time_of_Day}.Result
   )
  )
  MaterialFunctionCall(Emissive=MaterialFunctionCall(LDR Input=Clamp(None=@n16), HDR Tint=Add(A=@n18, B=@n23)) {materialFunction=&DeriveHDRfromLDR}.RemappedRange) {materialFunction=&MF_FX_EmissiveAdjustments}
  Multiply()
  n24 = ScalarParameter() {defaultValue=0.5, parameterName=Alpha Intensity}
  n25 = Multiply(A=ParticleColor().A, B=Add(A=Multiply(A=@n9, B=@n24), B=Add(A=@n12, B=@n6)))
  n26 = TextureCoordinate()
  n27 = Add(
   A=Multiply(A=ScalarParameter() {defaultValue=1, parameterName=Central_Alpha Intensity}, B=@n12)
   B=Power(Base=ComponentMask(None=@n26) {g=true}, Exp=ScalarParameter() {defaultValue=3.5, parameterName=Central Contast_1}) {constExponent=10}
  )
  n28 = Multiply(A=Multiply(A=Multiply(A=@n25, B=@n27), B=ScalarParameter() {defaultValue=1, parameterName=Opacity}), B=OneMinus(None=@n17))
  ShadingPathSwitch(Default=DepthFade(Opacity=@n28, FadeDistance=ScalarParameter() {parameterName=DepthFade}), Mobile=@n28) {desc=Migrated from FeatureLevelSwitch}
 ]
}
