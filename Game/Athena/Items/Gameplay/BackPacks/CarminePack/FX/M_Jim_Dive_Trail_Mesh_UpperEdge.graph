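# Translucent, unlit, two-sided mesh-particle material for the upper edge of the CarminePack backpack's dive trail.
# A UV sine and head/tail power falloff form a comet mask, combined with panning inky-smoke samples to blend dark and light colors.
# Time-of-day functions vary tail falloff and HDR glow.
# Alpha combines particle alpha, mask, scrolling streaks and inverse Fresnel, with dynamic parameters driving scroll speed, streak exponent and HDR intensity.
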Material /Game/Athena/Items/Gameplay/BackPacks/CarminePack/FX/M_Jim_Dive_Trail_Mesh_UpperEdge
graph {
 properties {
  bUsedWithMeshParticles=true
  bUsedWithNiagaraMeshParticles=true
  blendMode=BLEND_Translucent
  pixelDepthOffsetMode=PDOM_Legacy
  shadingModel=MSM_Unlit
  twoSided=true
  usedShadingModels=Unlit
 }
 nodes [
  n1 = TextureCoordinate()
  n2 = Sine(None=ComponentMask(None=@n1) {r=true}) {period=2}
  n3 = ComponentMask(None=@n1) {g=true}
  n4 = ScalarParameter() {defaultValue=0.5, parameterName=Comet_Tail_Exp}
  n5 = ScalarParameter() {defaultValue=0.75, parameterName=Comet_Tail_Exp_1}
  n6 = ScalarParameter() {defaultValue=0.5, parameterName=Comet_Tail_Exp_2}
  n7 = ScalarParameter() {defaultValue=1.25, parameterName=Comet_Tail_Exp_3}
  n8 = Multiply(
   A=Power(
    Base=@n3
    Exp=MaterialFunctionCall(Color_Morning=@n4, Color_Day=@n5, Color_Evening=@n6, Color_Night=@n7) {materialFunction=&Color_Time_of_Day}.Result
   )
   B=Power(Base=OneMinus(None=@n3), Exp=ScalarParameter() {defaultValue=3.1257, parameterName=Comet_Head_Exp})
  )
  n9 = Clamp(None=Multiply(A=Multiply(A=@n2, B=@n8)) {constB=2})
  n10 = TextureCoordinate() {uTiling=0.7, vTiling=0.125}
  n11 = Multiply(A=Time(), B=ScalarParameter() {defaultValue=0.6, parameterName=PanSpeed})
  n12 = TextureCoordinate() {uTiling=0.6, vTiling=0.125}
  n13 = LinearInterpolate(
   A=TextureSample(UVs=Panner(Coordinate=@n10, Time=@n11) {speedX=2, speedY=3.5}) {texture=&T_Inky_Smoke_Tile_Chromatic}.RGB
   B=TextureSample(UVs=Panner(Coordinate=@n12, Time=@n11) {speedX=-5, speedY=8}) {texture=&T_Inky_Smoke_Tile_Chromatic}.RGB
  )
  n14 = ScalarParameter() {parameterName=Streaks_LerpA}
  n15 = ScalarParameter() {defaultValue=1, parameterName=Streaks_LerpB}
  n16 = OneMinus(None=ComponentMask(None=TextureCoordinate()) {g=true})
  n17 = Multiply(A=@n9, B=Power(Base=Multiply(A=@n13) {constB=1.5}, Exp=LinearInterpolate(A=@n14, B=@n15, Alpha=@n16) {constA=0.6, constB=4}))
  Add(A=@n9, B=@n17)
  n18 = TextureCoordinate() {uTiling=0.25}
  n19 = TextureCoordinate() {uTiling=0.25}
  n20 = Add(
   A=TextureSample(UVs=Panner(Coordinate=@n18, Time=@n11) {speedX=-1, speedY=2.5}) {texture=&T_Inky_Smoke_Tile_Chromatic}.RGB
   B=TextureSample(UVs=Panner(Coordinate=@n19, Time=@n11) {speedX=1, speedY=5}) {texture=&T_Inky_Smoke_Tile_Chromatic}.RGB
  )
  n21 = LinearInterpolate(
   A=VectorParameter() {parameterName=DarkColor, defaultValue=<rgba 0.4 0.1473 0.0297 0>}.RGB
   B=VectorParameter() {parameterName=LightColor, defaultValue=<rgba 1.5 1.1175 0.0964 0>}.RGB
   Alpha=@n17
  )
  n22 = Multiply(A=Multiply(A=@n20, B=@n20), B=@n21)
  Clamp(None=@n22)
  n23 = Add(A=Clamp(None=@n21) {maxDefault=10}, B=@n22)
  n24 = MaterialFunctionCall(
   Color_Morning=ScalarParameter() {defaultValue=2, parameterName=HDRMult_4}
   Color_Day=ScalarParameter() {defaultValue=10, parameterName=HDRMult_3}
   Color_Evening=ScalarParameter() {defaultValue=5, parameterName=HDRMult_2}
   Color_Night=ScalarParameter() {defaultValue=2, parameterName=HDRMult_1}
  ) {materialFunction=&Color_Time_of_Day}
  n25 = DynamicParameter() {defaultValue=<rgba 0.1 1 1 1>, paramNames [Rotation Control, StretchControl, Streak_Opacity, HDR_Intensity]}
  MaterialFunctionCall(Emissive=MaterialFunctionCall(LDR Input=Clamp(None=@n23), HDR Tint=Multiply(A=@n23, B=@n24.Result), HDR Intensity Multiplier=@n25.HDR_Intensity) {materialFunction=&DeriveHDRfromLDR}.RemappedRange) {materialFunction=&MF_FX_EmissiveAdjustments}
  n26 = ParticleColor()
  n27 = Time()
  n28 = TextureSample(UVs=Panner(Coordinate=TextureCoordinate() {vTiling=0.1}, Time=Multiply(A=@n27, B=@n25.Rotation Control)) {speedX=0.5, speedY=0.25}) {texture=&T_Smooth_tile}
  n29 = ScalarParameter() {defaultValue=0.15, parameterName=Fresnel Exp}
  Multiply(
   A=Multiply(A=Multiply(A=@n26.A, B=@n9), B=Power(Base=@n28.R, Exp=@n25.Streak_Opacity))
   B=OneMinus(None=Fresnel(ExponentIn=@n29) {baseReflectFraction=0})
  )
  ScalarParameter() {defaultValue=2, parameterName=HDRMult}
  TextureCoordinate() {vTiling=0.7}
  TextureCoordinate() {vTiling=0.6}
  VectorParameter() {parameterName=DebrisColor, defaultValue=<rgba 3.2956 3.6779 7 0>}
 ]
}
