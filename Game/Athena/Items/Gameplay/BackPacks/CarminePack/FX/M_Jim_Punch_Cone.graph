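# Translucent, unlit, two-sided mesh-particle material for a smoky punch cone effect on the CarminePack backpack.
# Blended panning smoke layers, tinted by particle color and a V gradient, feed hue-shifted emissive.
# Opacity combines contrasted smoke, caustics, UV masks and optional inverted Fresnel, with depth and camera fades skipped on mobile; a dynamic parameter moves a soft mask.
# Layered time-driven sine waves wobble vertices, weighted along V.
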Material /Game/Athena/Items/Gameplay/BackPacks/CarminePack/FX/M_Jim_Punch_Cone
graph {
 properties {
  bUsedWithMeshParticles=true
  bUsedWithNiagaraMeshParticles=true
  blendMode=BLEND_Translucent
  pixelDepthOffsetMode=PDOM_Legacy
  shadingModel=MSM_Unlit
  twoSided=true
  usedShadingModels=Unlit
 }
 nodes [
  n1 = ScalarParameter() {defaultValue=-0.15, group=Color, parameterName="Hue Shift Percentage (S)"}
  n2 = OneMinus(None=ComponentMask(None=TextureCoordinate()) {g=true})
  n3 = Multiply(A=@n1, B=Clamp(None=Power(Base=@n2)))
  n4 = ParticleColor()
  n5 = ComponentMask(None=TextureCoordinate()) {g=true}
  n6 = Clamp(None=Power(Base=OneMinus(None=@n5)))
  n7 = LinearInterpolate(
   A=Multiply(A=VectorParameter() {group=Color, parameterName=ColorMultiply-LerpA, defaultValue=<rgba 1 0.5 0 1>}.RGB, B=@n4.RGB)
   B=@n4.RGB
   Alpha=@n6
  )
  n8 = Multiply(A=Time(), B=ScalarParameter() {defaultValue=1, parameterName=Speed})
  n9 = Panner(Coordinate=TextureCoordinate() {uTiling=3, vTiling=0.25}, Time=@n8) {speedX=-0.5, speedY=-1}
  n10 = Panner(Coordinate=TextureCoordinate() {uTiling=6, vTiling=0.5}, Time=@n8) {speedX=1, speedY=-2}
  n11 = TextureCoordinate() {uTiling=4, vTiling=0.5}
  n12 = TextureCoordinate() {uTiling=8}
  n13 = LinearInterpolate(
   A=TextureSample(UVs=Panner(Coordinate=@n11, Time=@n8) {speedX=-0.5, speedY=-1}) {texture=&T_Inky_Smoke_Tile_Chromatic}.RGB
   B=TextureSample(UVs=Panner(Coordinate=@n12, Time=@n8) {speedX=1, speedY=-2}) {texture=&T_Inky_Smoke_Tile_Chromatic}.RGB
  )
  n14 = LinearInterpolate(
   A=LinearInterpolate(
    A=TextureSample(UVs=@n9) {texture=&T_Inky_Smoke_Tile_Chromatic}.RGB
    B=TextureSample(UVs=@n10) {texture=&T_Inky_Smoke_Tile_Chromatic}.RGB
   )
   B=@n13
  )
  n15 = Multiply(A=Desaturation(None=@n14, Fraction=ScalarParameter() {defaultValue=-0.5, group=Color, parameterName=Saturation-Fraction})) {constB=1.5}
  MaterialFunctionCall(Emissive=MaterialFunctionCall(
    Hue Shift Percentage=StaticSwitchParameter(True=@n3, False=@n1) {defaultValue=true, parameterName=UseHueShiftGradient}
    Texture=Multiply(A=@n7, B=@n15)
   ) {materialFunction=&HueShift}.Result) {materialFunction=&MF_FX_EmissiveAdjustments}
  n16 = MaterialFunctionCall(Distance=ScalarParameter() {defaultValue=250, parameterName=Camera_FadeDistance}) {materialFunction=&Near_Camera_Fade}
  n17 = Fresnel() {baseReflectFraction=0.01, exponent=0.01}
  n18 = ScalarParameter() {defaultValue=2, parameterName=Fresnel-Exp}
  n19 = OneMinus(None=Clamp(None=Power(Base=@n17, Exp=@n18)))
  n20 = Constant() {r=1}
  n21 = Constant()
  n22 = Constant() {r=0.5}
  n23 = ComponentMask(None=TextureCoordinate()) {g=true}
  n24 = OneMinus(None=LinearInterpolate(B=@n13, Alpha=Multiply(A=@n23) {constB=0.25}))
  n25 = Multiply(A=@n14, B=Clamp(None=Power(Base=@n24) {constExponent=500}))
  n26 = Clamp(None=Power(Base=Multiply(A=@n23, B=@n23)) {constExponent=0.25})
  n27 = Power(
   Base=Add(A=LinearInterpolate(A=@n21, B=@n22, Alpha=@n6), B=LinearInterpolate(A=@n14, B=@n25, Alpha=@n26))
   Exp=ScalarParameter() {defaultValue=3, parameterName=PowerContrastTexture}
  ) {constExponent=0.7}
  n28 = SphereMask(
   A=TextureCoordinate() {uTiling=2, vTiling=0.7}
   B=ComponentMask(None=VectorParameter() {parameterName=MainMask_Offset, defaultValue=<rgba 1 0.65 0 1>}.RGB) {g=true, r=true}
   Radius=ScalarParameter() {defaultValue=0.35, parameterName=MainMask_Radius}
   Hardness=ScalarParameter() {defaultValue=0.5, parameterName=MainMask_Hardness}
  )
  n29 = Multiply(A=@n28, B=@n28) {constB=15}
  n30 = LinearInterpolate(A=Constant() {r=1}, B=Constant() {r=-1}, Alpha=Clamp(None=Power(Base=@n29) {constExponent=0.25}))
  n31 = ComponentMask(None=TextureCoordinate()) {g=true}
  n32 = OneMinus(None=Multiply(A=@n31, B=@n31))
  n33 = TextureSample(UVs=Panner(Coordinate=TextureCoordinate() {uTiling=4, vTiling=0.25}) {speedX=0.5, speedY=-1}) {texture=&T_ShieldCaustics}
  n34 = Add(A=@n32, B=@n32)
  n35 = Multiply(A=Multiply(A=Clamp(None=@n30), B=@n32), B=LinearInterpolate(A=Add(A=@n33.G, B=@n33.G), B=Constant() {r=1}, Alpha=Clamp(None=@n34)))
  n36 = Multiply(
   A=ParticleColor().A
   B=Multiply(A=StaticSwitchParameter(True=@n19, False=@n20) {defaultValue=true, parameterName=UseFresnel}, B=Multiply(A=@n27, B=@n35))
  )
  n37 = DepthFade(Opacity=@n36, FadeDistance=ScalarParameter() {defaultValue=250, parameterName=Depth_FadeDistance})
  n38 = DynamicParameter() {defaultValue=<rgba 0 0.5 1 1>, paramNames [PannerX,PannerY,Param3,Param4]}
  Multiply(
   A=ShadingPathSwitch(Default=Multiply(A=@n16.Result, B=@n37), Mobile=@n36) {desc=Migrated from FeatureLevelSwitch}
   B=SphereMask(
    A=TextureCoordinate() {uTiling=0}
    B=AppendVector(A=@n38.PannerX, B=@n38.PannerX)
    Radius=ScalarParameter() {defaultValue=0.35, parameterName=SoftMask_Radius}
    Hardness=ScalarParameter() {defaultValue=-1, parameterName=SoftMask_Hardness}
   )
  )
  n39 = Multiply(A=Frac(None=Time())) {constB=50}
  n40 = OneMinus(None=ComponentMask(None=TextureCoordinate()) {g=true})
  n41 = Multiply(A=Multiply(A=@n40, B=@n40)) {constB=-2}
  n42 = Multiply(A=Sine(None=Add(A=@n39, B=@n41))) {constB=1.2, desc=Amplitude}
  n43 = Add(A=Time(), B=ComponentMask(None=TextureCoordinate()) {g=true})
  n44 = TextureCoordinate()
  n45 = Add(A=Multiply(A=Time()) {constB=150}, B=Multiply(A=ComponentMask(None=@n44) {g=true}) {constB=2})
  n46 = Add(A=Multiply(A=Sine(None=@n43) {period=1.31}) {desc=Amplitude}, B=Multiply(A=Sine(None=@n45) {period=1.31}) {desc=Amplitude})
  n47 = Multiply(A=VertexNormalWS(), B=Max(A=Add(A=@n42, B=@n46)) {constB=-1})
  n48 = Time()
  n49 = AppendVector(A=ComponentMask(None=WorldPosition().XYZ) {b=true}, B=Constant2Vector())
  n50 = Add(A=Multiply(A=Frac(None=@n48)) {constB=32}, B=Frac(None=Multiply(A=@n49) {constB=-0.02}))
  n51 = Constant() {r=0.5}
  n52 = Multiply(A=Sine(None=@n50), B=@n51) {desc=Amplitude}
  n53 = Multiply(A=Cosine(None=@n50), B=@n51)
  n54 = OneMinus(None=ComponentMask(None=TextureCoordinate()) {g=true})
  n55 = Multiply(
   A=MaterialFunctionCall(X=ComponentMask(None=@n52) {r=true}, Y=ComponentMask(None=@n53) {r=true}, Z=Constant()) {materialFunction=&MakeFloat3}.Result
   B=Multiply(A=@n54, B=@n54)
  )
  n56 = Add(A=ComponentMask(None=TextureCoordinate()) {g=true}) {constB=0.5}
  Reroute(None=Multiply(A=Add(A=@n47, B=@n55) {constB=0}, B=Clamp(None=@n56)))
 ]
}
