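# Unlit, two-sided translucent mesh-particle material for swirling smoke FX on spline meshes in the CarminePack backpack effects.
# SplineThicken sets strand width, optionally modulated by an animated sine along V.
# Two panning inky smoke samples, sped and offset by dynamic parameters, are summed and tinted by particle color for desaturatable emissive.
# Alpha combines particle alpha and smoke green with a switchable mask: per-edge UV contrast gradients or a soft flare texture.
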Material /Game/Athena/Items/Gameplay/BackPacks/CarminePack/FX/M_Jim_Spline_Swirls
graph {
 properties {
  bUsedWithMeshParticles=true
  bUsedWithNiagaraMeshParticles=true
  blendMode=BLEND_Translucent
  pixelDepthOffsetMode=PDOM_Legacy
  shadingModel=MSM_Unlit
  twoSided=true
  usedShadingModels=Unlit
 }
 nodes [
  n1 = Add(A=ComponentMask(None=TextureCoordinate()) {g=true}, B=Time())
  MaterialFunctionCall(
   WidthBase=ScalarParameter() {defaultValue=50, parameterName="WidthBase (S)"}
   WidthTip=ScalarParameter() {defaultValue=50, parameterName="WidthTip (S)"}
   UVs for Thickness=StaticSwitchParameter(True=Sine(None=@n1) {period=2}, False=TextureCoordinate()) {parameterName=USe Sine Wave for Thickness}
  ) {materialFunction=&SplineThicken}
  n2 = ParticleColor()
  n3 = AppendVector(A=ScalarParameter() {defaultValue=1, parameterName=U_Scale}, B=ScalarParameter() {defaultValue=1, parameterName=V_Scale})
  n4 = Multiply(A=TextureCoordinate() {uTiling=0.5}, B=@n3)
  n5 = DynamicParameter() {paramNames [Param1,Param2,PanSpeed,RandomOffset]}
  n6 = Multiply(A=Time(), B=@n5.PanSpeed)
  n7 = TextureSample(UVs=Panner(Coordinate=Add(A=@n4, B=@n5.RandomOffset), Time=@n6) {speedX=0.05, speedY=1}) {texture=&T_Inky_Smoke_Tile_Chromatic}
  n8 = Multiply(A=TextureCoordinate() {vTiling=1.5}, B=@n3)
  n9 = TextureSample(UVs=Panner(Coordinate=Add(A=@n8, B=@n5.RandomOffset), Time=@n6) {speedX=-0.1, speedY=1}) {texture=&T_Inky_Smoke_Tile_Chromatic}
  n10 = Add(A=@n7.RGB, B=@n9.RGB)
  MaterialFunctionCall(Emissive=Desaturation(None=Multiply(A=@n2.RGB, B=@n10), Fraction=ScalarParameter() {parameterName=Desaturation})) {materialFunction=&MF_FX_EmissiveAdjustments}
  n11 = MaterialFunctionCall(In=ComponentMask(None=TextureCoordinate()) {r=true}, Contrast=ScalarParameter() {parameterName=U_LeftRight_Contrast}) {materialFunction=&CheapContrast}
  n12 = TextureCoordinate()
  n13 = MaterialFunctionCall(In=OneMinus(None=ComponentMask(None=@n12) {r=true}), Contrast=ScalarParameter() {parameterName=U_RightLeft_Contrast}) {materialFunction=&CheapContrast}
  n14 = MaterialFunctionCall(In=ComponentMask(None=TextureCoordinate()) {g=true}, Contrast=ScalarParameter() {parameterName=V_TopBottom_Contrast}) {materialFunction=&CheapContrast}
  n15 = TextureCoordinate()
  n16 = MaterialFunctionCall(In=OneMinus(None=ComponentMask(None=@n15) {g=true}), Contrast=ScalarParameter() {parameterName=V_BottomTop_Contrast}) {materialFunction=&CheapContrast}
  n17 = Multiply(
   A=Multiply(A=Multiply(A=@n11.Result, B=@n13.Result), B=Multiply(A=@n14.Result, B=@n16.Result))
   B=ScalarParameter() {defaultValue=100, parameterName=Contrast_Adjustment}
  )
  n18 = TextureSample() {texture=/Game/Global/Textures/Default/Masks_and_Gradients/T_New_Soft_Flair}
  n19 = ScalarParameter() {parameterName=AlphaContrast}
  Multiply(
   A=Multiply(A=@n2.A, B=Add(A=@n7.G, B=@n9.G))
   B=StaticSwitchParameter(
    True=Clamp(None=@n17)
    False=MaterialFunctionCall(In=@n18.G, Contrast=@n19) {materialFunction=&CheapContrast}.Result
   ) {parameterName=Use Individual Contrast Control}
  )
 ]
}
