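# Translucent, unlit, two-sided particle mesh material for the CarminePack backpack FX, drawing a scrolling inky-smoke effect, likely on a torus.
# Two panned smoke samples plus a vertical gradient, tinted by particle color, feed HDR emissive adjustments.
# Opacity combines that mask with particle alpha, inverted Fresnel, depth fade and an erosion term.
# Dynamic parameters drive erosion, pan speed, UV randomization and normal-based mesh expansion, which is summed with world-space vector-noise distortion.
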Material /Game/Athena/Items/Gameplay/BackPacks/CarminePack/FX/M_Jim_Torus
graph {
 properties {
  bUsedWithMeshParticles=true
  bUsedWithNiagaraMeshParticles=true
  blendMode=BLEND_Translucent
  pixelDepthOffsetMode=PDOM_Legacy
  shadingModel=MSM_Unlit
  twoSided=true
  usedShadingModels=Unlit
 }
 nodes [
  n1 = DynamicParameter() {paramNames [ExpandMesh,Erode,PanSpeed,RandomUV]}
  n2 = Add(A=TextureCoordinate() {uTiling=2, vTiling=0.1}, B=@n1.RandomUV)
  n3 = Multiply(A=@n1.PanSpeed, B=Time())
  n4 = Add(A=@n1.RandomUV, B=TextureCoordinate() {uTiling=4, vTiling=0.2})
  n5 = Add(
   A=TextureSample(UVs=Panner(Coordinate=@n2, Time=@n3) {speedX=0.1, speedY=-1}) {texture=&T_Inky_Smoke_Tile_Chromatic}.RGB
   B=TextureSample(UVs=Panner(Coordinate=@n4, Time=@n3) {speedX=0.1, speedY=-1}) {texture=&T_Inky_Smoke_Tile_Chromatic}.RGB
  )
  n6 = Power(
   Base=OneMinus(None=MaterialFunctionCall() {materialFunction=&LinearGradient}.VGradient)
   Exp=ScalarParameter() {defaultValue=2, parameterName=GradientControl}
  )
  Add(A=@n5, B=OneMinus(None=@n6))
  n7 = WorldPosition()
  Add(A=@n7.XYZ, B=Multiply(A=Time(), B=ScalarParameter() {defaultValue=5, parameterName=3DnoiseSpeed}))
  n8 = Multiply(A=@n7.XYZ, B=Multiply(A=Constant() {r=0.001}, B=ScalarParameter() {defaultValue=3, parameterName=NoiseScale}))
  Add(
   A=Multiply(A=VectorNoise(World Position=@n8) {noiseFunction=VNF_VectorALU}, B=ScalarParameter() {defaultValue=25, parameterName=NoiseDistortAmount})
   B=Multiply(A=VertexNormalWS(), B=DynamicParameter() {paramNames [ExpandMesh,Erode,PanSpeed,RandomUV]}.ExpandMesh)
  )
  n9 = Clamp(None=Add(A=@n5, B=@n6))
  n10 = ParticleColor()
  n11 = Multiply(A=@n9, B=@n10.RGB)
  MaterialFunctionCall(Emissive=MaterialFunctionCall(LDR Input=Clamp(None=@n11), HDR Tint=@n11) {materialFunction=&DeriveHDRfromLDR}.RemappedRange) {materialFunction=&MF_FX_EmissiveAdjustments}
  n12 = ComponentMask(None=@n9) {g=true}
  n13 = Multiply(A=@n12, B=Multiply(A=@n10.A, B=@n6))
  n14 = ScalarParameter() {defaultValue=1, parameterName=FresnelExp}
  n15 = MaterialFunctionCall(In=OneMinus(None=Fresnel(ExponentIn=@n14)), Contrast=ScalarParameter() {defaultValue=0.1, parameterName=FresnelContrast}) {materialFunction=&CheapContrast}
  n16 = ScalarParameter() {defaultValue=1, parameterName=ErodeHardness}
  n17 = Multiply(A=ComponentMask(None=Clamp(None=@n5)) {g=true}, B=@n16) {constB=2}
  n18 = Multiply(A=DynamicParameter() {paramNames [ExpandMesh,Erode,PanSpeed,RandomUV]}.Erode, B=@n16) {constB=2}
  Multiply(
   A=DepthFade(Opacity=Multiply(A=@n13, B=@n15.Result), FadeDistance=ScalarParameter() {defaultValue=64, parameterName=DepthFadeAmount})
   B=Clamp(None=Add(A=@n17, B=@n18))
  )
  Reroute(None=@n12)
 ]
}
